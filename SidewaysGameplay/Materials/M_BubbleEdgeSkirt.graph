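# Unlit, two-sided translucent material for the Sideways bubble's edge skirt, a smoky band placed where the bubble's sphere meets the landscape.
# World position offset uses runtime virtual texture landscape height and sphere-intersection math to resize, lift and flare the skirt.
# Distorted, dither-jittered panning noise erodes a UV gradient that indexes a curve-atlas color, alongside a depth fade.
# A shading-path switch collapses vertices to the object position on mobile.
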Material /SidewaysGameplay/Materials/M_BubbleEdgeSkirt
graph {
 properties {blendMode=BLEND_Translucent, pixelDepthOffsetMode=PDOM_Legacy, shadingModel=MSM_Unlit, twoSided=true, usedShadingModels=Unlit}
 nodes [
  n1 = MaterialFunctionCall(Float3=ObjectPositionWS()) {materialFunction=&BreakOutFloat3Components}
  n2 = NamedRerouteUsage()
  Add(A=@n1.B, B=@n2)
  Constant() {r=1}
  n3 = TextureCoordinate()
  n4 = ChannelMaskParameter(None=Subtract(A=Multiply(A=@n3) {constB=2})) {maskChannel=Green, parameterName=G, defaultValue=<rgba 0 1 0 0>}
  n5 = OneMinus(None=Abs(None=@n4.R))
  DepthFade(Opacity=@n5, FadeDistance=ScalarParameter() {parameterName=DepthFade_Distance})
  n6 = ObjectRadius()
  n7 = Distance(A=@n2, B=ComponentMask(None=ActorPositionWS()) {b=true})
  n8 = SquareRoot(None=Subtract(A=Multiply(A=@n6, B=@n6), B=Multiply(A=@n7, B=@n7)))
  n9 = Multiply(A=Subtract(A=ObjectRadius(), B=@n8)) {constB=-1}
  n10 = Divide(A=@n8, B=ObjectRadius())
  LinearInterpolate(A=@n9, B=Constant() {r=1}, Alpha=@n10)
  LinearInterpolate()
  n11 = TextureCoordinate()
  n12 = Multiply(
   A=AppendVector(A=ScalarParameter() {defaultValue=50, parameterName=X_Tiling}, B=ScalarParameter() {defaultValue=1, parameterName=Y_Tiling})
   B=Constant() {r=1}
  )
  n13 = Time()
  n14 = AppendVector(A=ScalarParameter() {parameterName=X_NoisePanSpeed}, B=ScalarParameter() {defaultValue=-0.1, parameterName=Y_NoisePanSpeed})
  n15 = AppendVector(
   A=ScalarParameter() {defaultValue=25, parameterName=NoiseDistort_X_Tiling}
   B=ScalarParameter() {defaultValue=1, parameterName=NoiseDistort_Y_Tiling}
  )
  n16 = Constant() {r=1}
  n17 = ScalarParameter() {defaultValue=0.025, parameterName=NoiseDistort_X_NoisePanSpeed}
  n18 = ScalarParameter() {defaultValue=-0.025, parameterName=NoiseDistort_Y_NoisePanSpeed}
  n19 = Add(A=Multiply(A=TextureCoordinate(), B=Multiply(A=@n15, B=@n16)), B=Multiply(A=Time(), B=AppendVector(A=@n17, B=@n18)))
  n20 = Subtract(A=Multiply(A=TextureSampleParameter2D(UVs=@n19) {parameterName=DistortionNoise, texture=&T_NoisePack_01}.RGB) {constB=2})
  n21 = ScalarParameter() {defaultValue=0.1, parameterName=NoiseDistortion_Strength}
  n22 = Multiply(
   A=MaterialFunctionCall() {materialFunction=&DitherTemporalAA}.Result
   B=ScalarParameter() {defaultValue=1, parameterName=DitherStrength}
  )
  n23 = Add(
   A=Add(A=Multiply(A=@n11, B=@n12), B=Multiply(A=@n13, B=@n14))
   B=Multiply(A=ComponentMask(None=@n20) {g=true, r=true}, B=Multiply(A=@n21, B=@n22))
  )
  n24 = Multiply(A=ChannelMaskParameter(None=TextureSampleParameter2D(UVs=@n23) {parameterName=SmokeyNoise, texture=&T_NoisePack_01}.RGB) {maskChannel=Green, parameterName=NoiseChannelMask, defaultValue=<rgba 0 1 0 0>}.R) {constB=0.5}
  n25 = CurveAtlasRowParameter(CurveTime=Clamp(None=Subtract(A=@n5, B=@n24))) {
   atlas=/Game/Athena/Apollo/Environments/Landscape/Materials/Corruption/CorruptionCurveAtlas
   curve=/Game/Athena/Apollo/Environments/Landscape/Materials/Corruption/SidewaysEdgeColorCurve
   defaultValue=8
   parameterName=ChromaLUT
  }
  Multiply(A=@n25)
  Multiply(A=GetMaterialAttributes() {attributeGetTypes [#1]}.Base Color) {constB=2}
  n26 = Subtract(A=@n1.B, B=@n2)
  n27 = Normalize(VectorInput=Subtract(
    A=WorldPosition().XYZ
    B=MaterialFunctionCall(X=@n1.R, Y=@n1.R, Z=@n26) {materialFunction=&MakeFloat3}.Result
   ))
  n28 = Multiply(A=@n27, B=@n9)
  n29 = MaterialFunctionCall(Float3=@n27) {materialFunction=&BreakOutFloat3Components}
  n30 = NamedRerouteUsage()
  n31 = ComponentMask(None=ActorPositionWS()) {b=true}
  n32 = MaterialFunctionCall(X=@n29.R, Y=@n29.R, Z=Add(A=Subtract(A=@n30, B=@n31), B=ScalarParameter() {defaultValue=24, parameterName=WorldHeightBias})) {materialFunction=&MakeFloat3}
  n33 = Multiply(A=ScalarParameter() {defaultValue=256, parameterName=Skirt_Scale}, B=@n4.R)
  NamedRerouteDeclaration(None=Add(A=Add(A=@n28, B=@n32.Result), B=Multiply(A=@n27, B=@n33))) {name=WPO, nodeColor=<rgba 0.5234 1 0 1>}
  NamedRerouteDeclaration(None=RuntimeVirtualTextureSample!(UVs, World Position) {materialType=WorldHeight, virtualTexture=/Game/Athena/Apollo/Maps/Landscape/VT/RVT_Apollo_Landscape_Height}.WorldHeight) {name=WorldHeight, nodeColor=<rgba 0 1 0.2734 1>}
  NamedRerouteDeclaration(None=@n22) {name=DitheredAA, nodeColor=<rgba 1 0.3984 0 1>}
  n34 = ComponentMask(None=ActorPositionWS()) {b=true}
  OneMinus(None=Divide(A=Subtract(A=@n34, B=@n2), B=ObjectRadius()))
  ScalarParameter() {defaultValue=0.1, parameterName=Noise}
  ShadingPathSwitch(Default=NamedRerouteUsage(), Mobile=Subtract(A=ObjectPositionWS(), B=WorldPosition().XYZ))
  Subtract(A=Multiply(A=@n10) {constB=2})
 ]
}
