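# Unlit two-sided material for the Sideways bubble, an animated, iridescent membrane shell.
# Panning distorted smoke, swirl and caustic noise feed curve-atlas colours, layered with fake sun scattering, specular, time-of-day tinting, cubemap reflection and audio-reactive inner glow.
# Camera-proximity sphere masks build darkening and portal opacity, while an intersect location and start time drive an expanding chromatic pulse.
# World position offset combines sine wobble with a camera push using collection-supplied bubble location and scale.
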Material /SidewaysGameplay/Materials/M_SidewaysBubble
graph {
 properties {bCastRayTracedShadows=false, pixelDepthOffsetMode=PDOM_Legacy, shadingModel=MSM_Unlit, twoSided=true, usedShadingModels=Unlit}
 nodes [
  Ceil(None=VertexColor().A)
  n1 = Multiply(A=Time()) {constB=0.25}
  Clamp(None=OneMinus(None=MaterialFunctionCall(Float=@n1) {materialFunction=&MF_Hash}.Float))
  n2 = NamedRerouteUsage()
  Clamp(None=Power(Base=Multiply(A=@n2) {constB=4}))
  n3 = CollectionParameter() {
   collection=&FortniteMaterialParameters
   parameterName=SidewaysBubble_Scale_LocalP2
  }
  Clamp(None=Round(None=@n3))
  n4 = CollectionParameter() {
   collection=&FortniteMaterialParameters
   parameterName=SunlightDirectionalVector
  }
  ComponentMask(None=@n4) {b=true, g=true, r=true}
  Constant3Vector() {constant=<rgba 1 0.4117 0 0>}
  Constant3Vector()
  n5 = DotProduct(A=SkyAtmosphereLightDirection(), B=Multiply(A=NamedRerouteUsage()) {constB=-1})
  ConstantBiasScale(None=@n5)
  Constant()
  Constant()
  Constant() {r=1}
  Constant() {r=1}
  Constant() {r=0.01}
  Constant() {r=0.001}
  Constant() {r=1}
  n6 = MaterialFunctionCall() {materialFunction=&ObjectPivotPoint}
  Divide(A=@n6.Object Pivot Location)
  n7 = CollectionParameter() {
   collection=&FortniteMaterialParameters
   parameterName=SunlightDirectionalVector
  }
  DotProduct(
   A=MaterialFunctionCall() {materialFunction=&CameraDirectionVector}.Result
   B=ComponentMask(None=@n7) {b=true, g=true, r=true}
  )
  n8 = NamedRerouteUsage()
  n9 = NamedRerouteUsage()
  n10 = ScalarParameter() {defaultValue=0.1, parameterName=VertexColor_Offset}
  n11 = LinearInterpolate(A=Clamp(None=@n8) {minDefault=0.4}, B=Clamp(None=Add(A=@n9, B=@n10)), Alpha=Clamp(None=@n8))
  n12 = Multiply(A=NamedRerouteUsage(), B=ScalarParameter() {defaultValue=0.25, parameterName=BubbleNoiseStrength})
  n13 = CurveAtlasRowParameter(CurveTime=Clamp(None=Subtract(A=@n11, B=@n12))) {
   atlas=/Game/Athena/Apollo/Environments/Landscape/Materials/Corruption/CorruptionCurveAtlas
   curve=/Game/Athena/Apollo/Environments/Landscape/Materials/Corruption/SidewaysBubbleColorCurve01
  }
  n14 = ScalarParameter() {defaultValue=0.25, parameterName=VertexColor_Offset_02}
  n15 = LinearInterpolate(
   A=VectorParameter() {parameterName=Linear_Color_01, defaultValue=<rgba 1 0.3419 0.0144 1>}.RGB
   B=VectorParameter() {parameterName=Linear_Color_02, defaultValue=<rgba 0.0545 0.0185 0.1356 1>}.RGB
   Alpha=Clamp(None=Add(A=@n9, B=@n14))
  )
  n16 = NamedRerouteUsage()
  n17 = Constant3Vector() {constant=<rgba 0.8879 0.5442 0.2309 1>}
  n18 = OneMinus(None=NamedRerouteUsage())
  n19 = Power(Base=OneMinus(None=Fresnel() {exponent=1})) {constExponent=3}
  n20 = Multiply(A=NamedRerouteUsage()) {constB=0.25}
  n21 = Power(Base=OneMinus(None=Fresnel() {exponent=1})) {constExponent=3}
  n22 = CollectionParameter() {collection=/Game/Athena/Events/AudioAnalysis/MPC_Audio_Reactive_Parameters, parameterName=Amplitude_AGV}
  n23 = OneMinus(None=@n22)
  n24 = Clamp(None=Subtract(A=@n21, B=Clamp(None=@n23) {maxDefault=0.25}))
  n25 = Multiply(A=@n22, B=ScalarParameter() {defaultValue=1, parameterName=FireFightInnerGlow_Brightness})
  n26 = Clamp(None=Add(A=Subtract(A=@n19, B=@n20), B=Multiply(A=@n24, B=@n25)))
  n27 = LinearInterpolate(
   A=LinearInterpolate(A=LinearInterpolate(A=@n13, B=@n15, Alpha=@n16), B=NamedRerouteUsage(), Alpha=NamedRerouteUsage())
   B=@n17
   Alpha=Power(Base=Multiply(A=@n18, B=@n26))
  )
  n28 = LinearInterpolate(
   A=@n27
   B=Multiply(A=@n27, B=VectorParameter() {parameterName=MembraneColor, defaultValue=<rgba 0.4323 0.4323 0.4323 1>}.RGB)
   Alpha=NamedRerouteUsage()
  )
  LinearInterpolate(
   A=Multiply(A=@n28, B=VectorParameter() {parameterName=UnlitSide_Multiplier, defaultValue=<rgba 0.55 0.55 0.55 1>}.RGB)
   B=Multiply(A=@n28, B=VectorParameter() {parameterName=LitSide_Multiplier, defaultValue=<rgba 2 2 2 1>}.RGB)
   Alpha=NamedRerouteUsage()
  )
  n29 = ScalarParameter() {defaultValue=0.01, parameterName=UV_Scale}
  n30 = MaterialFunctionCall() {materialFunction=/Engine/Functions/Engine_MaterialFunctions02/Utility/FOV}
  n31 = Constant() {r=0.414}
  n32 = Constant() {r=1.091}
  n33 = Constant()
  n34 = Constant() {r=1}
  LinearInterpolate(
   A=Multiply(A=@n29) {constB=2}
   B=@n29
   Alpha=Saturate(None=MaterialFunctionCall(Input=@n30."FOV ", Input Low=@n31, Input High=@n32, Target Low=@n33, Target High=@n34) {materialFunction=&RemapValueRange}.Result)
  )
  n35 = WorldPosition()
  n36 = ComponentMask(None=CameraPositionWS()) {b=true, g=true, r=true}
  n37 = Subtract(A=ObjectPositionWS(), B=@n36)
  n38 = Multiply(A=@n37) {constB=-1}
  n39 = NamedRerouteUsage()
  n40 = Multiply(A=Distance(A=@n36, B=@n35.XYZ), B=Normalize(VectorInput=LinearInterpolate(A=@n38, B=@n37, Alpha=@n39)))
  n41 = ScalarParameter() {defaultValue=128, parameterName=PortalSize}
  n42 = NamedRerouteUsage()
  n43 = OneMinus(None=@n42)
  n44 = OneMinus(None=SphereMask(A=@n35.XYZ, B=Add(A=@n36, B=@n40), Radius=Multiply(A=@n41, B=@n43), Hardness=ScalarParameter() {parameterName=Portal_Hardness}))
  n45 = Multiply(A=NamedRerouteUsage()) {constB=0.1}
  n46 = ScalarParameter() {defaultValue=0.5, parameterName=MinOpacity}
  LinearInterpolate(A=Clamp(None=Subtract(A=@n44, B=@n45), Min=@n46) {minDefault=1}, B=Constant() {r=1}, Alpha=@n42)
  n47 = ScalarParameter() {defaultValue=512, parameterName=PushDistance}
  n48 = CollectionParameter() {collection=&FortniteMaterialParameters, parameterName=PlayerPosition}
  LinearInterpolate(
   A=@n47
   B=Distance(A=CameraPositionWS(), B=ComponentMask(None=@n48) {b=true, g=true, r=true})
   Alpha=ScalarParameter() {defaultValue=1, parameterName=DisableWPOSplitScreen}
  )
  n49 = AppendVector(A=ScalarParameter() {defaultValue=4, parameterName=X_Tiling}, B=ScalarParameter() {defaultValue=2, parameterName=Y_Tiling})
  n50 = Multiply(
   A=Time()
   B=AppendVector(A=ScalarParameter() {parameterName=X_NoisePanSpeed}, B=ScalarParameter() {defaultValue=0.03, parameterName=Y_NoisePanSpeed})
  )
  n51 = Add(A=Multiply(A=NamedRerouteUsage(), B=@n49), B=@n50)
  n52 = ScalarParameter() {defaultValue=4, parameterName=MicroTile_Multiplier}
  n53 = AppendVector(
   A=ScalarParameter() {defaultValue=4, parameterName=NoiseDistort_X_Tiling}
   B=ScalarParameter() {defaultValue=2, parameterName=NoiseDistort_Y_Tiling}
  )
  n54 = Multiply(A=NamedRerouteUsage(), B=@n53)
  n55 = Multiply(A=@n54, B=@n52)
  n56 = Multiply(
   A=Time()
   B=AppendVector(
    A=ScalarParameter() {defaultValue=0.025, parameterName=NoiseDistort_X_NoisePanSpeed}
    B=ScalarParameter() {defaultValue=-0.025, parameterName=NoiseDistort_Y_NoisePanSpeed}
   )
  )
  n57 = Multiply(A=NamedRerouteUsage(), B=@n53)
  n58 = Multiply(A=@n57, B=@n52)
  n59 = LinearInterpolate(
   A=TextureSampleParameter2D(UVs=Add(A=@n55, B=@n56)) {parameterName=DistortionNoise, texture=&T_NoisePack_01}.RGB
   B=TextureSampleParameter2D(UVs=Add(A=@n58, B=@n56)) {parameterName=DistortionNoise, texture=&T_NoisePack_01}.RGB
   Alpha=NamedRerouteUsage()
  )
  n60 = ComponentMask(None=Subtract(A=Multiply(A=@n59) {constB=2})) {g=true, r=true}
  n61 = Multiply(
   A=MaterialFunctionCall() {materialFunction=&DitherTemporalAA}.Result
   B=ScalarParameter() {defaultValue=1, parameterName=DitherStrength}
  )
  n62 = Constant()
  n63 = ShadingPathSwitch(
   Default=QualitySwitch(Default=@n61, Low=@n62, High=@n61, Medium=@n61, Epic=@n61)
   Mobile=QualitySwitch(Default=@n62, Low=@n62, High=@n61, Medium=@n61, Epic=@n61)
  )
  n64 = Multiply(A=@n60, B=Multiply(A=ScalarParameter() {defaultValue=0.1, parameterName=NoiseDistortion_Strength}, B=@n63))
  n65 = TextureSampleParameter2D(UVs=Add(A=Multiply(A=@n51, B=@n52), B=@n64)) {parameterName=SmokeyNoise, texture=&T_NoisePack_01}
  n66 = Add(A=Multiply(A=NamedRerouteUsage(), B=@n49), B=@n50)
  n67 = Add(A=Multiply(A=@n66, B=@n52), B=@n64)
  n68 = TextureSampleParameter2D(UVs=@n67) {parameterName=SmokeyNoise, texture=&T_NoisePack_01}
  n69 = NamedRerouteUsage()
  n70 = LinearInterpolate(
   A=TextureSampleParameter2D(UVs=Add(A=@n54, B=@n56)) {parameterName=DistortionNoise, texture=&T_NoisePack_01}.RGB
   B=TextureSampleParameter2D(UVs=Add(A=@n57, B=@n56)) {parameterName=DistortionNoise, texture=&T_NoisePack_01}.RGB
   Alpha=NamedRerouteUsage()
  )
  n71 = ComponentMask(None=Subtract(A=Multiply(A=@n70) {constB=2})) {g=true, r=true}
  n72 = Multiply(A=@n71, B=Multiply(A=ScalarParameter() {defaultValue=0.1, parameterName=NoiseDistortion_Strength}, B=@n63))
  n73 = TextureSampleParameter2D(UVs=Add(A=@n51, B=@n72)) {parameterName=SmokeyNoise, texture=&T_NoisePack_01}
  n74 = TextureSampleParameter2D(UVs=Add(A=@n66, B=@n72)) {parameterName=SmokeyNoise, texture=&T_NoisePack_01}
  n75 = NamedRerouteUsage()
  n76 = LinearInterpolate(
   A=ChannelMaskParameter(None=LinearInterpolate(A=@n65.RGB, B=@n68.RGB, Alpha=@n69)) {maskChannel=Green, parameterName=NoiseChannelMask, defaultValue=<rgba 0 1 0 0>}.R
   B=ChannelMaskParameter(None=LinearInterpolate(A=@n73.RGB, B=@n74.RGB, Alpha=@n75)) {maskChannel=Green, parameterName=NoiseChannelMask, defaultValue=<rgba 0 1 0 0>}.R
   Alpha=NamedRerouteUsage()
  )
  LinearInterpolate(B=@n76, Alpha=NamedRerouteUsage()) {constB=0}
  n77 = CurveAtlasRowParameter(CurveTime=NamedRerouteUsage()) {
   atlas=&CorruptionCurveAtlas
   curve=/Game/Athena/Apollo/Environments/Landscape/Materials/Corruption/SidewaysbubbleSSCurve
   defaultValue=2
   parameterName=SS_ScatteringCurve
  }
  n78 = Multiply(A=@n28, B=@n77)
  n79 = VectorParameter() {parameterName=SS_Inscattering_Color, defaultValue=<rgba 0.8906 0.4149 0 1>}
  n80 = LinearInterpolate(
   A=Multiply(A=@n28, B=Multiply(A=@n77, B=@n77))
   B=Multiply(A=Multiply(A=@n78, B=@n79.RGB), B=ScalarParameter() {defaultValue=4, parameterName=SS_BrightnessIncreas})
   Alpha=NamedRerouteUsage()
  )
  n81 = NamedRerouteUsage()
  n82 = ScalarParameter() {defaultValue=1, parameterName=CubeMapColorIntensity}
  n83 = Power(Base=NamedRerouteUsage())
  n84 = Multiply(A=Power(Base=Fresnel() {exponent=0.5}), B=Multiply(A=Clamp(None=@n83)) {constB=1.5})
  n85 = LinearInterpolate(
   A=LinearInterpolate(A=Multiply(A=@n80), B=NamedRerouteUsage(), Alpha=NamedRerouteUsage())
   B=LinearInterpolate(A=@n80, B=Multiply(A=@n81, B=@n82), Alpha=Clamp(None=@n84))
   Alpha=NamedRerouteUsage()
  )
  n86 = NamedRerouteUsage()
  n87 = LinearInterpolate(A=@n85, B=NamedRerouteUsage(), Alpha=@n86)
  n88 = ScalarParameter() {defaultValue=8, parameterName=SpecularBrightness}
  n89 = VectorParameter() {parameterName=MorningEveningTimeBrightness, defaultValue=<rgba 0.4583 0.4583 0.4583 1>}
  n90 = Multiply(
   A=LinearInterpolate(A=@n87, B=Multiply(A=@n87, B=@n88), Alpha=NamedRerouteUsage())
   B=MaterialFunctionCall(
    Color_Morning=@n89.RGB
    Color_Day=VectorParameter() {parameterName=DayTimeBrightnes, defaultValue=<rgba 1 1 1 1>}.RGB
    Color_Evening=@n89.RGB
    Color_Night=VectorParameter() {parameterName=NightTimeBrightness, defaultValue=<rgba 0.5104 0.5104 0.5104 1>}.RGB
   ) {materialFunction=&Color_Time_of_Day}.Result
  )
  n91 = ScalarParameter() {defaultValue=0.35, parameterName=PassthroughDarkenFactor}
  n92 = SphereMask(
   A=WorldPosition().XYZ
   B=CameraPositionWS()
   Radius=ScalarParameter() {defaultValue=400, parameterName=DarkenRadius}
   Hardness=ScalarParameter() {defaultValue=0.5, parameterName=DarkenHArdness}
  )
  n93 = TextureSample(UVs=Panner(Coordinate=TextureCoordinate() {uTiling=23, vTiling=23}) {speedY=0.1}) {texture=&T_NoisePack_01}
  n94 = Multiply(A=Subtract(A=Multiply(A=@n93.RGB) {constB=2}), B=ScalarParameter() {defaultValue=32, parameterName=RippleDistortionStrength})
  n95 = WorldPosition()
  n96 = VectorParameter() {parameterName=PawnIntersectLocation, defaultValue=<rgba -19310.6934 20948.8672 587.2242 1>}
  n97 = ScalarParameter() {defaultValue=1024, parameterName=BrightnessPulseRadius}
  n98 = Time()
  n99 = ScalarParameter() {defaultValue=10000000000, parameterName=GameTimeFromBP}
  n100 = Saturate(None=Divide(A=Subtract(A=@n98, B=@n99), B=ScalarParameter() {defaultValue=1, parameterName=PulseDuration}))
  n101 = Multiply(
   A=SphereMask(A=Add(A=@n94, B=@n95.XYZ), B=ComponentMask(None=@n96.RGB) {b=true, g=true, r=true}, Radius=Multiply(A=@n97, B=@n100), Hardness=Constant())
   B=OneMinus(None=@n92)
  )
  n102 = CurveAtlasRowParameter(CurveTime=@n101) {
   atlas=&CorruptionCurveAtlas
   curve=/Game/Athena/Apollo/Environments/Landscape/Materials/Corruption/ChromaticColorCurve
   defaultValue=1
   parameterName=ChromaticCurve
  }
  n103 = ScalarParameter() {defaultValue=2, parameterName=PulseBrightness}
  n104 = Multiply(A=@n101, B=OneMinus(None=@n100))
  LinearInterpolate(
   A=LinearInterpolate(A=@n90, B=Multiply(A=@n90, B=@n91), Alpha=@n92)
   B=Power(Base=Multiply(A=@n102, B=@n103)) {constExponent=1}
   Alpha=Clamp(None=Power(Base=@n104))
  ) {constA=1, constB=10}
  n105 = CollectionParameter() {
   collection=&FortniteMaterialParameters
   parameterName=SunlightDirectionalVector
  }
  n106 = Add(A=CameraVectorWS(), B=ComponentMask(None=@n105) {b=true, g=true, r=true})
  n107 = ScalarParameter() {parameterName=SpecMinPower}
  n108 = ScalarParameter() {defaultValue=250, parameterName=SpecMaxPower}
  n109 = Clamp(None=Multiply(A=NamedRerouteUsage()) {constB=1.5})
  n110 = Power(Base=OneMinus(None=NamedRerouteUsage())) {constExponent=12}
  n111 = Power(
   Base=DotProduct(A=Normalize(VectorInput=@n106), B=NamedRerouteUsage())
   Exp=LinearInterpolate(A=LinearInterpolate(A=@n107, B=@n108, Alpha=@n109) {constA=12, constB=32}, B=@n108, Alpha=Clamp(None=@n110))
  ) {constExponent=6}
  MaterialFunctionCall(In=@n111, Contrast=ScalarParameter() {defaultValue=1, parameterName=SpecContrast}) {materialFunction=&CheapContrast}
  n112 = ReflectionVectorWS(CustomWorldNormal=NamedRerouteUsage())
  n113 = TextureSampleParameterCube!(UVs=@n112, Bias, Apply View MipBias) {
   mipValueMode=TMVM_MipBias
   parameterName=MorningCube
   texture=/Game/Packages/Fortress_Sky/SkyDome/MaterialInstances/SkyDome_Apollo/Cubemap_Apollo_Morning
  }
  n114 = TextureSampleParameterCube(
   UVs=@n112
   Bias=LinearInterpolate(
    A=ScalarParameter() {defaultValue=5, parameterName=Cube_MipBias_02}
    B=ScalarParameter() {defaultValue=2, parameterName=Cube_MipBias}
    Alpha=NamedRerouteUsage()
   )
  ) {mipValueMode=TMVM_MipBias, parameterName=DayCube, texture=/Game/FN_PublicTurntable/TurntableAssets/Cubemap_Apollo_Day}
  MaterialFunctionCall(
   Color_Morning=@n113.RGB
   Color_Day=@n114.RGB
   Color_Evening=TextureSampleParameterCube!(UVs=@n112, Apply View MipBias) {parameterName=EveningCube, texture=/Game/Packages/Fortress_Sky/SkyDome/MaterialInstances/SkyDome_Apollo/Cubemap_Apollo_Night}.RGB
   Color_Night=TextureSampleParameterCube(UVs=@n112) {parameterName=EveningCube, texture=&Cubemap_Apollo_Night}.RGB
  ) {materialFunction=&Color_Time_of_Day}
  MaterialFunctionCall!(Color) {materialFunction=/Game/Packages/Fortress_SharedMaterials/Anime_Master_Material/MF_TODColorCurve}
  MaterialFunctionCall() {materialFunction=&ObjectPivotPoint}
  n115 = Subtract(A=Distance(A=CameraPositionWS(), B=WorldPosition().XYZ)) {constB=32}
  Multiply(A=@n46, B=Clamp(None=Divide(A=@n115) {constB=32}))
  Multiply(A=@n85, B=VectorParameter() {parameterName=EdgeColor, defaultValue=<rgba 1 0 0.1355 1>}.RGB)
  n116 = CurveAtlasRowParameter(CurveTime=@n86) {
   atlas=&CorruptionCurveAtlas
   curve=/Game/Athena/Apollo/Environments/Landscape/Materials/Corruption/SidewaysEdgeColorCurve
   defaultValue=8
   parameterName=ChromaLUT
  }
  Multiply(A=@n116)
  Multiply(B=LinearInterpolate() {constA=1, constB=0})
  n117 = NamedRerouteUsage()
  n118 = AppendVector(
   A=ScalarParameter() {defaultValue=16, parameterName=BubblePattern_X_Tiling}
   B=ScalarParameter() {defaultValue=8, parameterName=BubblePattern_Y_Tiling}
  )
  n119 = Time()
  n120 = AppendVector(A=ScalarParameter() {parameterName=X_NoisePanSpeed}, B=ScalarParameter() {defaultValue=0.03, parameterName=Y_NoisePanSpeed})
  n121 = TextureSample(UVs=Add(A=Multiply(A=@n117, B=@n118), B=Multiply(A=@n119, B=@n120))) {texture=/CorruptionGameplay/Textures/T_Caustics_Pattern1}
  n122 = TextureSampleParameter2D(UVs=Panner(Coordinate=TextureCoordinate() {uTiling=2, vTiling=2}) {speedY=0.25}) {parameterName=SmokeyNoise, texture=&T_NoisePack_01}
  Multiply(A=Power(Base=Multiply(A=@n121.G) {constB=5}), B=@n122.G)
  n123 = MaterialFunctionCall() {materialFunction=&ObjectPivotPoint}
  n124 = WorldPosition()
  n125 = Ceil(None=DotProduct(A=CameraVectorWS(), B=VertexNormalWS()))
  Multiply(
   A=Multiply(A=Subtract(A=@n123.Object Pivot Location, B=@n124.XYZ), B=OneMinus(None=@n125))
   B=Truncate!(None=NamedRerouteUsage()) {is=class MaterialExpressionTruncate}
  )
  Multiply(A=@n17) {constB=5}
  Multiply(A=Time(), B=AppendVector(A=ScalarParameter() {parameterName=Membrane_01_XPan}, B=ScalarParameter() {parameterName=Membrane_01_YPan}))
  Multiply(A=Time(), B=AppendVector(A=ScalarParameter() {parameterName=Membrane_02_XPan}, B=ScalarParameter() {parameterName=Membrane_02_YPan}))
  Multiply(A=@n111) {constB=10}
  n126 = Power(Base=@n5) {constExponent=4}
  n127 = VectorParameter() {parameterName=Diffuse_SunColor, defaultValue=<rgba 1 0.0942 0 1>}
  Multiply(A=@n126, B=@n127.RGB)
  n128 = Fresnel(ExponentIn=ScalarParameter() {defaultValue=2, parameterName=FresnelExponent}) {exponent=0.5}
  NamedRerouteDeclaration(None=Multiply(A=Multiply(A=@n128, B=@n128)) {constB=0.75}) {name=Fresnel, nodeColor=<rgba 0.1953 0 1 1>}
  n129 = AppendVector(
   A=ScalarParameter() {defaultValue=1, parameterName=Bubble_Caustic_X_Tiling}
   B=ScalarParameter() {defaultValue=1, parameterName=Bubble_Caustic_Y_Tiling}
  )
  n130 = Multiply(A=NamedRerouteUsage(), B=@n129)
  n131 = Multiply(
   A=Time()
   B=AppendVector(A=ScalarParameter() {parameterName=Bubble_Caustic_X_pan}, B=ScalarParameter() {defaultValue=0.025, parameterName=Bubble_Caustic_Y_Pan})
  )
  n132 = Multiply(A=NamedRerouteUsage(), B=Multiply(A=NamedRerouteUsage(), B=ScalarParameter() {defaultValue=0.025, parameterName=CausticWarpStrength}))
  n133 = TextureSampleParameter2D(UVs=Add(A=Add(A=@n130, B=@n131), B=@n132)) {parameterName=Param, texture=/Game/Global/Textures/Default/Caustics/T_Caustics_Softer}
  n134 = NamedRerouteUsage()
  n135 = Add(A=Add(A=Multiply(A=@n134, B=@n129), B=@n131), B=@n132)
  n136 = TextureSampleParameter2D(UVs=@n135) {parameterName=Param, texture=&T_Caustics_Softer}
  n137 = NamedRerouteUsage()
  n138 = Clamp(None=Multiply(A=LinearInterpolate(A=@n133.G, B=@n136.G, Alpha=@n137)) {constB=2})
  n139 = NamedRerouteUsage()
  n140 = OneMinus(None=DotProduct(A=CameraVectorWS(), B=NamedRerouteUsage()))
  NamedRerouteDeclaration(None=Multiply(A=Multiply(A=@n138, B=@n139), B=@n140)) {name=CausticsNoise, nodeColor=<rgba 0.2656 0 1 1>}
  NamedRerouteDeclaration(None=@n28) {name=Colors, nodeColor=<rgba 0.4531 0 1 1>}
  n141 = CameraPositionWS()
  n142 = VertexNormalWS()
  n143 = Abs(None=@n142)
  n144 = Divide(A=MaterialFunctionCall() {materialFunction=&ObjectScale}.Scale XYZ)
  n145 = Distance(
   A=Subtract(A=Multiply(A=@n141, B=@n143), B=Multiply(A=@n142, B=@n144))
   B=Multiply(
    A=@n143
    B=MaterialFunctionCall() {materialFunction=&ObjectPivotPoint}.Object Pivot Location
   )
  )
  n146 = MaterialFunctionCall(Float3=VertexNormalWS()) {materialFunction=&BreakOutFloat3Components}
  n147 = MaterialFunctionCall(X=@n146.R, Y=@n146.R, Z=Abs(None=@n146.B)) {materialFunction=&MakeFloat3}
  n148 = MaterialFunctionCall(Float3=MaterialFunctionCall() {materialFunction=&CameraDirectionVector}.Result) {materialFunction=&BreakOutFloat3Components}
  n149 = Constant() {r=1}
  n150 = Multiply(A=Normalize(VectorInput=MaterialFunctionCall(X=@n148.R, Y=@n148.R, Z=@n149) {materialFunction=&MakeFloat3}.Result)) {constB=-1}
  n151 = LinearInterpolate(
   A=ScalarParameter() {defaultValue=128, parameterName=DistanceDivisorA}
   B=ScalarParameter() {defaultValue=512, parameterName=DistanceDivisorB}
   Alpha=Abs(None=DotProduct(A=@n147.Result, B=@n150))
  )
  NamedRerouteDeclaration(None=Saturate(None=Divide(A=@n145, B=@n151))) {name=DistanceCheck, nodeColor=<rgba 0.0312 1 0 1>}
  NamedRerouteDeclaration(None=LinearInterpolate(A=@n60, B=@n71, Alpha=NamedRerouteUsage())) {name=Distorted_UV, nodeColor=<rgba 1 0.3516 0 1>}
  NamedRerouteDeclaration(None=Clamp(None=Subtract(A=@n121.G, B=@n122.G))) {name=BubblePattern, nodeColor=<rgba 1 0.2109 0 1>}
  n152 = Multiply(A=Multiply(A=@n140, B=NamedRerouteUsage())) {constB=2}
  n153 = CurveAtlasRowParameter(CurveTime=ConstantBiasScale(None=Sine(None=@n152))) {
   atlas=&CorruptionCurveAtlas
   curve=&ChromaticColorCurve
   defaultValue=1
   parameterName=ChromaticCurve
  }
  NamedRerouteDeclaration(None=Multiply(A=@n153) {constB=1.25}) {name=Caustics_Color, nodeColor=<rgba 0.3594 0 1 1>}
  n154 = Multiply(
   A=NamedRerouteUsage()
   B=AppendVector(A=ScalarParameter() {defaultValue=8, parameterName=Membrane_X_Tiling}, B=ScalarParameter() {defaultValue=4, parameterName=Membrane_Y_Tiling})
  )
  n155 = ScalarParameter() {defaultValue=0.002, parameterName=Membrane_MaroDistortion, sliderMin=1}
  n156 = Multiply(A=NamedRerouteUsage(), B=@n155) {constB=0.1}
  n157 = NamedRerouteUsage()
  n158 = ScalarParameter() {defaultValue=0.005, parameterName=MembraneDistortionStrength}
  n159 = ScalarParameter() {defaultValue=0.01, parameterName=Membrane_01PanX}
  n160 = ScalarParameter() {defaultValue=0.025, parameterName=Membrane_01Pany}
  n161 = Add(A=Add(A=Add(A=@n154, B=@n156), B=Multiply(A=@n157, B=@n158)), B=Multiply(A=Time(), B=AppendVector(A=@n159, B=@n160)))
  n162 = TextureSampleParameter2D!(UVs=@n161, Bias=Constant(), Apply View MipBias) {
   mipValueMode=TMVM_MipBias
   parameterName=Membrane_Texture01
   samplerType=SAMPLERTYPE_LinearColor
   texture=/Game/Athena/Apollo/Environments/Landscape/Materials/Corruption/T_Corruption_SwirlDetailPack
  }
  n163 = Add(A=Multiply(A=Add(A=@n154) {constB=0.25}) {constB=1.5}, B=Multiply(A=NamedRerouteUsage(), B=@n155) {constB=0.1})
  n164 = Multiply(A=NamedRerouteUsage(), B=@n158)
  n165 = Time()
  n166 = AppendVector(
   A=ScalarParameter() {defaultValue=-0.02, parameterName=Membrane_02PanX}
   B=ScalarParameter() {defaultValue=0.05, parameterName=Membrane_02Pany}
  )
  n167 = BumpOffset(
   Coordinate=Add(A=Add(A=@n163, B=@n164), B=Multiply(A=@n165, B=@n166))
   Height=Constant()
   HeightRatioInput=ScalarParameter() {defaultValue=0.5, parameterName=HeightRatio}
  )
  n168 = TextureSampleParameter2D(UVs=@n167, Bias=Constant()) {
   mipValueMode=TMVM_MipBias
   parameterName=Membrane_Texture
   samplerType=SAMPLERTYPE_LinearColor
   texture=&T_Corruption_SwirlDetailPack
  }
  n169 = Add(A=@n162.A, B=@n168.A)
  n170 = ScalarParameter() {defaultValue=3, parameterName=MembraneMipLevel}
  n171 = TextureSampleParameter2D(UVs=@n161, Bias=@n170) {
   mipValueMode=TMVM_MipBias
   parameterName=Membrane_Texture01
   samplerType=SAMPLERTYPE_LinearColor
   texture=&T_Corruption_SwirlDetailPack
  }
  n172 = TextureSampleParameter2D(UVs=@n167, Bias=@n170) {
   mipValueMode=TMVM_MipBias
   parameterName=Membrane_Texture
   samplerType=SAMPLERTYPE_LinearColor
   texture=&T_Corruption_SwirlDetailPack
  }
  n173 = Add(A=@n171.A, B=@n172.A)
  n174 = Power(Base=NamedRerouteUsage()) {constExponent=1.5}
  n175 = Clamp(None=Multiply(A=LinearInterpolate(A=@n169, B=@n173, Alpha=@n174)) {constB=2})
  n176 = Clamp(None=@n174)
  NamedRerouteDeclaration(None=LinearInterpolate(B=Multiply(A=@n175, B=@n176), Alpha=NamedRerouteUsage())) {name=MembraneNoise, nodeColor=<rgba 0 0.6484 1 1>}
  n177 = Distance(A=CameraPositionWS(), B=WorldPosition().XYZ)
  n178 = ScalarParameter() {defaultValue=7000, parameterName=FadeDistance}
  NamedRerouteDeclaration(None=Clamp(None=Divide(A=@n177, B=@n178))) {name=DistanceFade, nodeColor=<rgba 0.7812 1 0 1>}
  n179 = ComponentMask(None=SkyAtmosphereLightDirection()) {b=true, g=true, r=true}
  n180 = PixelNormalWS()
  NamedRerouteDeclaration(None=ConstantBiasScale(None=DotProduct(A=@n179, B=@n180))) {name=SS_Mask, nodeColor=<rgba 1 0.3984 0 1>}
  NamedRerouteDeclaration(None=@n63) {name=DitheredAA, nodeColor=<rgba 1 0.3984 0 1>}
  NamedRerouteDeclaration() {name=InteriorUVs, nodeColor=<rgba 1 0 0.375 1>}
  n181 = Multiply(
   A=NamedRerouteUsage()
   B=AppendVector(
    A=ScalarParameter() {defaultValue=16, parameterName=HighFreq_NoiseDistort_X_Tiling}
    B=ScalarParameter() {defaultValue=8, parameterName=HighFreq_NoiseDistort_Y_Tiling}
   )
  )
  n182 = Multiply(
   A=Time()
   B=AppendVector(
    A=ScalarParameter() {defaultValue=0.075, parameterName=HighFreq_NoiseDistort_X_PanSpeed}
    B=ScalarParameter() {defaultValue=-0.075, parameterName=HighFreq_NoiseDistort_Y_PanSpeed}
   )
  )
  n183 = Multiply(A=TextureSampleParameter2D!(UVs=Add(A=@n181, B=@n182), Apply View MipBias) {parameterName=DistortionNoise, texture=&T_NoisePack_01}.RGB) {constB=2}
  NamedRerouteDeclaration(None=ComponentMask(None=Subtract(A=@n183)) {g=true, r=true}) {name=HighFreqDistortUV, nodeColor=<rgba 0.8516 0 1 1>}
  n184 = TransformPosition(None=ActorPositionWS()) {transformSourceType=TRANSFORMPOSSOURCE_World, transformType=TRANSFORMPOSSOURCE_View}
  NamedRerouteDeclaration(None=Divide(A=ObjectRadius(), B=ChannelMaskParameter(None=@n184) {maskChannel=Blue, parameterName=Actorz, defaultValue=<rgba 0 0 1 0>}.R)) {name=ScreenPercentage, nodeColor=<rgba 1 0 0.2812 1>}
  NamedRerouteDeclaration(None=Clamp(None=TwoSidedSign())) {name="TwoSidedSign(0-1)", nodeColor=<rgba 0.1016 1 0 1>}
  n185 = Constant() {r=1}
  n186 = ScalarParameter() {defaultValue=0.0005, parameterName=InternalSunRadius}
  n187 = ScalarParameter() {defaultValue=0.5, parameterName=SunHardness}
  n188 = Power(Base=Multiply(A=SphereMask(A=@n5, B=@n185, Radius=@n186, Hardness=@n187)) {constB=5})
  NamedRerouteDeclaration(None=Clamp(None=Add(A=@n126, B=@n188) {constB=0})) {name=SunMask, nodeColor=<rgba 1 0.375 0 1>}
  n189 = Multiply(A=@n127.RGB, B=OneMinus(None=NamedRerouteUsage()))
  NamedRerouteDeclaration(None=Power(Base=Multiply(A=@n189) {constB=4})) {name=SunColor, nodeColor=<rgba 0 0.2031 1 1>}
  NamedRerouteDeclaration(None=Multiply(A=@n114.RGB, B=ScalarParameter() {defaultValue=1, parameterName=CubeMapBrightness_01})) {name=CubeMapColor, nodeColor=<rgba 0.7109 0 1 1>}
  n190 = ComponentMask(None=WorldPosition().XYZ) {b=true}
  n191 = RuntimeVirtualTextureSample!(UVs, World Position) {materialType=WorldHeight, virtualTexture=/Game/Athena/Apollo/Maps/Landscape/VT/RVT_Apollo_Landscape_Height}
  NamedRerouteDeclaration(None=Divide(A=Subtract(A=@n190, B=@n191.WorldHeight), B=ObjectRadius()) {constB=1024}) {name=RVT_Gradient, nodeColor=<rgba 0 0.2734 1 1>}
  n192 = Divide(A=Subtract(A=WorldPosition().XYZ, B=@n6.Object Pivot Location), B=ObjectRadius())
  n193 = ComponentMask(None=Abs(None=@n192)) {b=true}
  n194 = QualitySwitch(Default=NamedRerouteUsage(), Low=@n193)
  n195 = QualitySwitch(Default=@n193, Low=@n193, High=@n193, Medium=@n193)
  n196 = SphereMask(
   A=NamedRerouteDeclaration(None=ShadingPathSwitch(Default=@n194, Mobile=@n195)) {name=LocalZGradient, nodeColor=<rgba 0.6875 0 1 1>}
   B=Constant()
   Radius=ScalarParameter() {defaultValue=0.25, parameterName=EdgeSpread}
   Hardness=Constant()
  )
  n197 = Clamp(None=@n196)
  NamedRerouteDeclaration(None=Multiply(A=@n197, B=@n197)) {name=BorderEdgeMask, nodeColor=<rgba 1 0 0.9375 1>}
  NamedRerouteDeclaration() {name=RVT_Color, nodeColor=<rgba 1 0.6563 0 1>}
  n198 = Multiply(A=NamedRerouteUsage()) {constB=2}
  n199 = DotProduct(A=Multiply(A=Constant3Vector() {constant=<rgba 0 0 1 0>}, B=Subtract(A=@n198)) {constB=-1}, B=NamedRerouteUsage())
  NamedRerouteDeclaration(None=Abs(None=OneMinus(None=@n199))) {name=Z_Mask, nodeColor=<rgba 1 0 0.0469 1>}
  n200 = Multiply(A=@n111) {constB=0.75}
  n201 = OneMinus(None=NamedRerouteUsage())
  NamedRerouteDeclaration(None=Multiply(A=Clamp(None=@n200), B=Power(Base=@n201) {constExponent=6})) {name=FakeSpecular, nodeColor=<rgba 1 0.6563 0 1>}
  NamedRerouteDeclaration(None=TextureCoordinate()) {name=BubbleUV, nodeColor=<rgba 0.1953 0 1 1>}
  NamedRerouteDeclaration(None=SphereMask(
    A=NamedRerouteUsage()
    B=Constant() {r=1}
    Radius=ScalarParameter() {defaultValue=1, parameterName=TiPmaskRadius}
    Hardness=ScalarParameter() {defaultValue=0.75, parameterName=TipMaskHArdness}
   )) {name=TipMask, nodeColor=<rgba 0.6641 0 1 1>}
  n202 = Multiply(A=NamedRerouteUsage()) {constB=0.5}
  n203 = CurveAtlasRowParameter(CurveTime=Clamp(None=Subtract(A=@n196, B=@n202))) {
   atlas=&CorruptionCurveAtlas
   curve=&SidewaysEdgeColorCurve
   defaultValue=8
   parameterName=ChromaLUT
  }
  NamedRerouteDeclaration(None=@n203) {name=EdgeColor, nodeColor=<rgba 0 1 1 1>}
  n204 = CameraVectorWS()
  n205 = Transform(None=TextureSample(UVs=@n67) {samplerType=SAMPLERTYPE_Normal, texture=/Game/Gadgets/Assets/Telephone_Booth/Textures/T_BlindsNormalNoise_N}.RGB)
  NamedRerouteDeclaration(None=Clamp(None=DotProduct(A=@n204, B=@n205))) {name=SmokeyNoise_N, nodeColor=<rgba 1 0 0.8203 1>}
  NamedRerouteDeclaration(None=ShaderStageSwitch(PixelShader=PixelNormalWS(), VertexShader=VertexNormalWS())) {name=RotatedNormals, nodeColor=<rgba 1 0 0.75 1>}
  n206 = WorldPosition()
  n207 = ObjectRadius()
  NamedRerouteDeclaration(None=Divide(A=ComponentMask(None=@n206.XYZ) {g=true, r=true}, B=Multiply(A=@n207) {constB=2})) {name=UV_Z-Projected, nodeColor=<rgba 0 1 0.8125 1>}
  n208 = DotProduct(A=CameraVectorWS(), B=@n180)
  n209 = DotProduct(A=@n179, B=Multiply(A=CameraVectorWS()) {constB=-1})
  NamedRerouteDeclaration(None=Clamp(None=Multiply(A=@n208, B=@n209))) {name=SS_Inscatter_Mask, nodeColor=<rgba 0 0.4375 1 1>}
  NamedRerouteDeclaration(None=@n192) {name=NormalizedWorldPosition, nodeColor=<rgba 0 1 0.625 1>}
  n210 = ScalarParameter() {defaultValue=400, parameterName=PushDistanceBias}
  n211 = LinearInterpolate(Alpha=ScalarParameter() {parameterName=BubbleMask}) {constA=-1}
  n212 = Multiply(A=NamedRerouteUsage(), B=Multiply(A=Add(A=@n47, B=@n210), B=@n211))
  n213 = CollectionParameter() {
   collection=&FortniteMaterialParameters
   parameterName=SidewaysBubble_Location_LocalP1
  }
  n214 = CameraPositionWS()
  n215 = Distance(A=ComponentMask(None=@n213) {b=true, g=true, r=true}, B=@n214)
  n216 = CollectionParameter() {
   collection=&FortniteMaterialParameters
   parameterName=SidewaysBubble_Scale_LocalP1
  }
  n217 = Subtract(A=@n216, B=ScalarParameter() {defaultValue=1200, parameterName=Bias})
  n218 = Add(A=@n47, B=ScalarParameter() {defaultValue=300, parameterName=DarkenRadiusBias})
  n219 = ScalarParameter() {defaultValue=0.5, parameterName=DarkenHardness}
  n220 = WorldPosition()
  n221 = Constant() {r=0.5}
  n222 = DotProduct(
   A=MaterialFunctionCall() {materialFunction=&CameraDirectionVector}.Result
   B=VertexNormalWS()
  )
  n223 = Multiply(A=@n222) {constB=-1}
  n224 = Clamp(None=@n211)
  n225 = Multiply(
   A=Multiply(A=SphereMask(A=@n215, B=@n217, Radius=@n218, Hardness=@n219), B=SphereMask(A=@n220.XYZ, B=@n214, Radius=@n218, Hardness=@n221))
   B=Clamp(None=LinearInterpolate(A=@n222, B=@n223, Alpha=@n224))
  )
  n226 = Divide(A=NamedRerouteUsage(), B=VectorParameter() {parameterName=WobbleScale, defaultValue=<rgba 1 1 1 1>}.RGB)
  n227 = Time()
  n228 = Add(
   A=ChannelMaskParameter(None=@n226) {parameterName=X Pan}.R
   B=Multiply(A=@n227, B=ScalarParameter() {defaultValue=0.125, parameterName=X_PanSpeed})
  )
  n229 = ScalarParameter() {defaultValue=0.5, parameterName=Y_PanOffset}
  n230 = Add(
   A=ChannelMaskParameter(None=@n226) {maskChannel=Green, parameterName=Y_Pan, defaultValue=<rgba 0 1 0 0>}.R
   B=Multiply(A=Add(A=@n227, B=@n229), B=ScalarParameter() {defaultValue=0.075, parameterName=Y_PanSpeed})
  )
  n231 = ScalarParameter() {defaultValue=0.37, parameterName=Z_PanOffset}
  n232 = Add(
   A=ChannelMaskParameter(None=@n226) {maskChannel=Blue, parameterName=Z_Pan, defaultValue=<rgba 0 0 1 0>}.R
   B=Multiply(A=Add(A=@n227, B=@n231), B=ScalarParameter() {defaultValue=0.05, parameterName=Z_PanSpeed})
  )
  n233 = MaterialFunctionCall(X=ConstantBiasScale(None=Sine(None=@n228)), Y=ConstantBiasScale(None=Sine(None=@n230)), Z=ConstantBiasScale(None=Sine(None=@n232))) {materialFunction=&MakeFloat3}
  n234 = NamedRerouteUsage()
  n235 = ScalarParameter() {defaultValue=5, parameterName=MaxWPO_ScaleBias}
  n236 = Multiply(
   A=Multiply(A=VectorParameter() {parameterName=WPO_o, defaultValue=<rgba 256 256 90 1>}.RGB, B=Clamp(None=@n234, Max=@n235) {maxDefault=5})
   B=ScalarParameter() {defaultValue=1, parameterName=WPO_Scale}
  )
  n237 = Subtract(A=NamedRerouteUsage(), B=ScalarParameter() {defaultValue=0.25, parameterName=Z-Gradient-Offset})
  n238 = Multiply(A=Multiply(A=Multiply(A=@n233.Result, B=@n236), B=NamedRerouteUsage()), B=Multiply(A=NamedRerouteUsage(), B=Clamp(None=@n237)))
  n239 = Constant()
  n240 = QualitySwitch(Default=@n238, Low=@n239)
  n241 = QualitySwitch(Default=@n239, Low=@n239, High=@n238, Medium=@n238, Epic=@n238)
  NamedRerouteDeclaration(None=Add(A=LinearInterpolate(B=@n212, Alpha=@n225), B=ShadingPathSwitch(Default=@n240, Mobile=@n241))) {name=WPO, nodeColor=<rgba 1 0 0.5625 1>}
  n242 = PixelNormalWS()
  NamedRerouteDeclaration(None=Clamp(None=DotProduct(A=@n179, B=@n242))) {name=NdotL, nodeColor=<rgba 0 1 0.9297 1>}
  NamedRerouteDeclaration(None=Divide(A=ObjectRadius(), B=ScalarParameter() {defaultValue=75, parameterName=BaseObjectScale})) {name=ObjectScaleBias, nodeColor=<rgba 1 0 0.8203 1>}
  NamedRerouteDeclaration(None=@n76) {name=SmokeyNoise, nodeColor=<rgba 1 0.5391 0 1>}
  NamedRerouteUsage()
  NamedRerouteUsage()
  NamedRerouteUsage()
  NamedRerouteUsage()
  NamedRerouteUsage()
  NamedRerouteUsage()
  OneMinus(None=@n208)
  OneMinus(None=DotProduct(A=CameraVectorWS(), B=PixelNormalWS()))
  PixelNormalWS()
  PixelNormalWS()
  PixelNormalWS()
  ScalarParameter() {defaultValue=0.1, parameterName=Noise}
  ScalarParameter() {defaultValue=0.015, parameterName=MembraneDistortionStrength_02}
  ScalarParameter() {defaultValue=400, parameterName=DarkenRadius}
  ScalarParameter() {defaultValue=350, parameterName=BubblePushDistance}
  ScalarParameter() {defaultValue=512, parameterName=CameraMaskRadius}
  n243 = MaterialFunctionCall() {materialFunction=&ObjectPivotPoint}
  SphereMask(
   A=MaterialFunctionCall(
    UVs=MaterialFunctionCall!(Center Point World Position=@n243.Object Pivot Location) {materialFunction=/Engine/Functions/Engine_MaterialFunctions02/UVs/WorldSpaceAlignedScreenCoordinates}.Offset UVs
    Texture Scale=Constant() {r=0.002}
   ) {materialFunction=&ScaleUVsByCenter}.UVs
   B=Constant2Vector()
   Radius=Multiply(A=ObjectRadius())
   Hardness=Constant()
  )
  Subtract(A=@n21, B=@n20)
  n244 = Time()
  Subtract(A=TextureSample() {texture=&T_Caustics_Pattern1}.G, B=OneMinus(None=Frac(None=@n244)))
  TextureCoordinate()
  TextureCoordinate()
  TextureSampleParameter2D() {parameterName=SmokeyNoise, texture=&T_NoisePack_01}
  TextureSample() {texture=&T_NoisePack_01}
  TextureSample() {samplerType=SAMPLERTYPE_Masks, texture=/Game/Characters/Player/Male/Medium/Bodies/M_MED_Storm_Soldier/Textures/ElectricityNoise_M}
  TextureSample() {samplerType=SAMPLERTYPE_Masks, texture=/Game/Weapons/WeaponSkins/Wraps/Textures/PlasmaCharge/PlasmaCharge_M}
  TextureSample() {texture=/Game/Weapons/WeaponSkins/Wraps/Textures/PlasmaClouds/T_Wrap_PlasmaClouds__M}
  TextureSample() {texture=/Game/CharClasses/Outlander/Gadgets/AntiMaterialCharge/Materials/T_AntiMaterialCharge_Dissolve_01}
  TextureSample() {samplerType=SAMPLERTYPE_Masks, texture=/Game/Global/Textures/Default/Noise/T_CellularNoise}
  TextureSample() {samplerType=SAMPLERTYPE_Masks, texture=&ElectricityNoise_M}
  TextureSample() {texture=/Game/Global/Textures/Default/Caustics/T_Turbulence_Seamless}
  TextureSample(UVs=@n135) {texture=&T_ShieldCaustics}
  VectorParameter() {parameterName=BubblePatternColor, defaultValue=<rgba 0 0 0 1>}
  VectorParameter() {parameterName=HeadSocketLocation, defaultValue=<rgba 0 0 0 1>}
  WorldPosition()
 ]
}
