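# Material function that adds a noisy dissolve along the object's local bounds Z axis, with a glowing edge, behind the UseSidewaysFade static switch.
# Progress comes from a manual amount or from time since BP_FadeTimeBegin, scaled by bounds height and a position-hashed per-meter duration, optionally reversed.
# A smoothstep of bounding-box UVW, offset by triplanar cloud noise, multiplies Opacity Mask and drives the emissive edge blend on the incoming material attributes.
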MaterialFunction /SidewaysGameplay/Materials/MF_SidewaysFade
graph {nodes [
  ActorPositionWS()
  n1 = Time()
  n2 = MaterialFunctionCall() {materialFunction=&ObjectLocalBounds}
  n3 = Multiply(A=Multiply(A=ComponentMask(None=@n2.Local Bounds Size) {b=true}) {constB=0.01}, B=NamedRerouteUsage())
  Divide(A=Fmod(A=@n1, B=@n3), B=@n3)
  n4 = FunctionInput() {inputType=FunctionInput_MaterialAttributes}
  n5 = GetMaterialAttributes(None=@n4) {attributeGetTypes [#1, #2]}
  n6 = LinearInterpolate(A=@n5.EmissiveColor, B=NamedRerouteUsage(), Alpha=NamedRerouteUsage())
  n7 = Multiply(A=@n5.OpacityMask, B=NamedRerouteUsage())
  FunctionOutput(None=StaticSwitchParameter(
    True=SetMaterialAttributes!(MaterialAttributes=@n4, EmissiveColor=@n6, OpacityMask=@n7) {attributeSetTypes [#1, #2]}
    False=@n5.MaterialAttributes
   ) {group=SidewaysFade, parameterName=UseSidewaysFade, sortPriority=-200})
  n8 = ScalarParameter() {defaultValue=0.4, group=SidewaysFade, parameterName=FadeThreshold, sortPriority=-97}
  n9 = Multiply(
   A=NamedRerouteDeclaration(None=Multiply(A=@n8) {constB=0.5}) {name=HalfThreshold, nodeColor=<rgba 0.6172 0 1 1>}
   B=ScalarParameter() {group=SidewaysFade, parameterName=FadeExtentsBias, sortPriority=-95}
  )
  n10 = Subtract(A=@n1, B=NamedRerouteUsage())
  n11 = Saturate(None=Divide(A=NamedRerouteDeclaration(None=@n10) {name=FadeTime, nodeColor=<rgba 0.9453 0 1 1>}, B=@n3))
  n12 = LinearInterpolate(A=@n11, B=OneMinus(None=@n11), Alpha=ScalarParameter() {group=SidewaysFade, parameterName=ReverseFade, sortPriority=-95})
  n13 = ScalarParameter() {defaultValue=1, group=SidewaysFade, parameterName=ManualFadeAmount, sortPriority=-99}
  n14 = ScalarParameter() {defaultValue=1, group=SidewaysFade, parameterName=UseManualFade, sortPriority=-100}
  NamedRerouteDeclaration(None=LinearInterpolate(A=Subtract(B=@n9) {constA=0}, B=Add(B=@n9) {constA=1}, Alpha=LinearInterpolate(A=@n12, B=@n13, Alpha=@n14))) {name=BaseFadeAmount, nodeColor=<rgba 0 1 1 1>}
  n15 = TextureObject() {texture=&AtmosphericCloudNoise01}
  n16 = ScalarParameter() {defaultValue=512, group=SidewaysFade, parameterName=NoiseSize, sortPriority=30}
  n17 = WorldPosition()
  n18 = VertexNormalWS()
  n19 = ScalarParameter() {group=SidewaysFade, parameterName=NoiseTriplanarContrast, sortPriority=31}
  NamedRerouteDeclaration(None=Multiply(
    A=MaterialFunctionCall(TextureObject=@n15, TextureSize=@n16, WorldPosition=@n17.XYZ, World Space Normal=@n18, ProjectionTransitionContrast=@n19) {materialFunction=&WorldAlignedTexture}.XYZ Texture
    B=ScalarParameter() {defaultValue=2, group=SidewaysFade, parameterName=NoiseMultiplier}
   ) {constB=2}) {name=NoiseTexture, nodeColor=<rgba 0 0.6953 1 1>}
  n20 = NamedRerouteUsage()
  n21 = NamedRerouteUsage()
  n22 = Add(A=@n20, B=@n21)
  n23 = Subtract(A=@n20, B=@n21)
  n24 = MaterialFunctionCall() {materialFunction=&BoundingBoxBased_0-1_UVW}
  n25 = NamedRerouteUsage()
  NamedRerouteDeclaration(None=Power(Base=SmoothStep(Min=@n22, Max=@n23, Value=@n24.B) {constMax=0.7, constMin=0.3}, Exp=Add(A=@n25))) {name=OpacityMask, nodeColor=<rgba 0 0.6953 1 1>}
  n26 = VectorParameter() {group=SidewaysFade, parameterName=Emissivecolor, sortPriority=1, defaultValue=<rgba 0.4393 0.038 1 1>}
  NamedRerouteDeclaration(None=Multiply(A=@n26.RGB, B=ScalarParameter() {defaultValue=1, group=SidewaysFade, parameterName=EmissiveIntenisty, sortPriority=10})) {name=EmissiveColor, nodeColor=<rgba 0 0.4609 1 1>}
  n27 = ScalarParameter() {defaultValue=0.1, group=SidewaysFade, parameterName=EmissiveEdgeThreshold, sortPriority=12}
  n28 = ScalarParameter() {defaultValue=0.5, group=SidewaysFade, parameterName=EmissiveEdgeMin, sortPriority=11}
  NamedRerouteDeclaration(None=SmoothStep(Min=Add(A=@n27, B=@n28) {constB=0.5}, Max=@n28, Value=NamedRerouteUsage()) {constMax=0.5, constMin=0.6}) {name=EmissiveMask, nodeColor=<rgba 0 0.2734 1 1>}
  NamedRerouteDeclaration(None=ScalarParameter() {group=SidewaysFade, parameterName=BP_FadeTimeBegin, sortPriority=-90}) {name=FadeTimeBegin, nodeColor=<rgba 0 1 0.8359 1>}
  n29 = ObjectPositionWS()
  n30 = Sine(None=Add(A=ComponentMask(None=@n29) {r=true}, B=ComponentMask(None=@n29) {g=true}))
  NamedRerouteDeclaration(None=LinearInterpolate(
    A=ScalarParameter() {defaultValue=1, group=SidewaysFade, parameterName=FadeTimePerMeterMin, sortPriority=-98}
    B=ScalarParameter() {defaultValue=1, group=SidewaysFade, parameterName=FadeTimePerMeterMax, sortPriority=-98}
    Alpha=ConstantBiasScale(None=@n30)
   )) {name=FadeTimeLength, nodeColor=<rgba 0 1 0.9766 1>}
  ScalarParameter() {defaultValue=1, group=SidewaysFade, parameterName=FadeTimePerMeterMin, sortPriority=-98}
 ]}
