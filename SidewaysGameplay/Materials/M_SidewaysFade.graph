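# M_SidewaysFade is a masked, Default Lit base material for Sideways gameplay, applying a fade through MF_SidewaysFade.
# Tiled UVs drive diffuse, mask, normal and switchable, intensity-scaled emissive textures, packed into material attributes.
# The SpecularMasks green channel alone feeds metallic, specular and roughness.
# The attributes then chain through MF_ChestAnimation, MF_SidewaysFade and MF_GameWideShared, which layer shared animation, fade and game-wide effects before output.
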Material /SidewaysGameplay/Materials/M_SidewaysFade
graph {
 properties {bUseMaterialAttributes=true, blendMode=BLEND_Masked, pixelDepthOffsetMode=PDOM_Legacy, usedShadingModels=Default Lit}
 nodes [
  n1 = Multiply(A=TextureCoordinate(), B=ScalarParameter() {defaultValue=1, parameterName=Tiling})
  n2 = TextureSampleParameter2D(UVs=@n1) {
   group=Textures
   parameterName=Diffuse
   samplerSource=SSM_Wrap_WorldGroupSettings
   sortPriority=-5
   texture=/Game/Global/Textures/Default/Blanks/T_MidgreysRGB
  }
  n3 = TextureSampleParameter2D(UVs=@n1) {
   group=Textures
   parameterName=SpecularMasks
   samplerSource=SSM_Wrap_WorldGroupSettings
   samplerType=SAMPLERTYPE_Masks
   sortPriority=-5
   texture=&T_Fortnite_Default_S
  }
  n4 = MaterialFunctionCall(Float3=@n3.RGB) {materialFunction=&BreakOutFloat3Components}
  n5 = TextureSampleParameter2D(UVs=@n1) {
   group=Textures
   parameterName=Emissive
   samplerSource=SSM_Wrap_WorldGroupSettings
   sortPriority=-5
   texture=&T_BlacksRGB
  }
  n6 = ScalarParameter() {defaultValue=1, group=Textures, parameterName=EmissiveIntensity}
  n7 = TextureSampleParameter2D(UVs=@n1) {
   group=Textures
   parameterName=Normals
   samplerSource=SSM_Wrap_WorldGroupSettings
   samplerType=SAMPLERTYPE_Normal
   sortPriority=-5
   texture=/Engine/EngineMaterials/FlatNormal
  }
  n8 = MakeMaterialAttributes!(
   BaseColor=@n2.RGB
   Metallic=@n4.G
   Specular=@n4.G
   Roughness=@n4.G
   Anisotropy
   EmissiveColor=StaticSwitchParameter(True=Multiply(A=@n5.RGB, B=@n6), False=Constant3Vector()) {group=Textures, parameterName=UseEmissive}
   Opacity
   OpacityMask
   Normal=@n7.RGB
   Tangent
   WorldPositionOffset
   SubsurfaceColor
   ClearCoat
   ClearCoatRoughness
   AmbientOcclusion
   Refraction
   CustomizedUVs_0
   CustomizedUVs_1
   CustomizedUVs_2
   CustomizedUVs_3
   CustomizedUVs_4
   CustomizedUVs_5
   CustomizedUVs_6
   CustomizedUVs_7
   PixelDepthOffset
   ShadingModel
   Displacement
  )
  MaterialFunctionCall!(In=MaterialFunctionCall!(In=MaterialFunctionCall!(In=@n8) {materialFunction=/Game/Packages/Fortress_SharedMaterials/Base_Material_Functions/MF_ChestAnimation}.Result) {materialFunction=&/MF_SidewaysFade}.Result) {materialFunction=/Game/Athena/WinterArt/Materials/MF_GameWideShared}
 ]
}
